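# Exposed material function that samples an inscattering lookup texture laid out as a frame atlas, indexed by depth and refracted direction.
# RefractionWorldVector bends around the negated pixel normal; the result's normalized horizontal X and vertical component form the UV within a tile.
# In Depth picks the tile, passing through overridable CUV pins; color, raw gradient and refracted vector are output.
# Octahedral mapping and fog-density branches are left unconnected.
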MaterialFunction /Landmass/Landscape/OldPrototype_BP/Mats/MF/Inscattering_LUT
graph {
 properties {bExposeToLibrary=true}
 nodes [
  n1 = PixelNormalWS()
  n2 = MaterialFunctionCall!(
   Refractive Index 1
   Refractive Index 2
   Surface Normal=Multiply(A=FunctionInput(Preview=@n1) {bUsePreviewValueAsDefault=true, inputName=PixelNormal}) {constB=-1}
   Camera Vector
  ) {materialFunction=&/RefractionWorldVector}
  n3 = FunctionOutput(None=@n2.Refracted Vector) {outputName=Refracted Vector, sortPriority=3}
  ConstantBiasScale(None=Multiply(
    A=MaterialFunctionCall!(Vector=@n3) {materialFunction=/Engine/Functions/Engine_MaterialFunctions01/Vectors/UnitVectorToOctahedron}.Octahedron
    B=ScalarParameter() {defaultValue=0.9844, parameterName=UV Bias}
   ) {constB=0.99})
  n4 = FunctionInput() {bUsePreviewValueAsDefault=true, inputName=In Depth, inputType=FunctionInput_Scalar}
  n5 = ScalarParameter() {defaultValue=16, parameterName=LUT Frames}
  n6 = AppendVector(A=@n4, B=Multiply(A=@n4, B=Multiply(A=@n5, B=@n5)))
  n7 = MaterialFunctionCall(Float2=FunctionInput(Preview=FunctionOutput(None=@n6) {outputName=CUV Out}) {bUsePreviewValueAsDefault=true, inputName=CUV In, inputType=FunctionInput_Vector2, sortPriority=1}) {materialFunction=&BreakOutFloat2Components}
  FunctionOutput(None=@n7.R) {outputName=Raw Gradient, sortPriority=2}
  n8 = Multiply(A=@n3)
  n9 = Multiply(A=@n8, B=Constant3Vector() {constant=<rgba 1 1 0 0>})
  n10 = AppendVector(A=DotProduct(A=Normalize(VectorInput=@n9), B=Constant3Vector() {constant=<rgba -1 0 0 0>}), B=ComponentMask(None=@n8) {b=true})
  n11 = ScalarParameter() {defaultValue=1.1, parameterName=ZM}
  n12 = Divide(A=ConstantBiasScale(None=Divide(A=@n10, B=@n11)) {desc="Bias = 1, Scale = 0.5"}) {constB=16}
  n13 = Floor(None=@n7.G)
  n14 = Divide(A=@n13, B=@n5)
  n15 = Divide(
   A=MaterialFunctionCall(X=Fmod(A=@n13, B=@n5), Y=Floor(None=@n14)) {materialFunction=&MakeFloat2}.Result
   B=ScalarParameter() {defaultValue=16, parameterName=LUT Frames}
  )
  FunctionOutput(None=TextureSample!(UVs=Add(A=@n12, B=@n15) {constB=0}, Tex, Level=Constant(), Apply View MipBias) {mipValueMode=TMVM_MipLevel, samplerType=SAMPLERTYPE_LinearColor}.RGB) {sortPriority=1}
  LinearInterpolate(A=CameraVectorWS(), B=@n2.Lerp Approximation, Alpha=ScalarParameter() {parameterName=RRREFF})
  MaterialFunctionCall()
  Multiply(A=AtmosphericLightVector()) {constB=-1}
  n16 = WorldPosition()
  n17 = CollectionParameter() {parameterName=k0}
  n18 = Subtract(
   A=DotProduct(A=ComponentMask(None=@n16.XYZ) {g=true, r=true}, B=ComponentMask(None=@n17) {g=true, r=true})
   B=ScalarParameter() {defaultValue=750, parameterName=Fog Offset}
  )
  n19 = ScalarParameter() {defaultValue=15000, parameterName=Particle Gradient Length}
  Power(Base=Clamp(None=Divide(A=@n18, B=@n19)) {maxDefault=0.999}, Exp=ScalarParameter() {defaultValue=0.3, parameterName=Density Power})
  ScalarParameter() {parameterName=DebugPos}
 ]
}
