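# Unlit sky material for a 24-hour sky, layering stars, sun, moon and clouds with Sky Atmosphere lighting.
# Three procedural vector-noise star layers get blackbody colors, rotate around a tilted axis and are masked behind the moon.
# The sun is a sphere-mask disk, while the moon is a ray-traced, textured sphere with phase shading.
# Wind-scrolled noise on horizon-curved coordinates blends atmosphere-lit clouds over the sky.
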Material /DaySequence/M_24HrSky
graph {
 properties {bHasPixelAnimation=true, bIsSky=true, shadingModel=MSM_Unlit, usedShadingModels=Unlit}
 nodes [
  CameraPositionWS()
  ComponentMask(None=SkyAtmosphereAerialPerspective!(World Position)) {b=true, g=true, r=true}
  n1 = ScalarParameter() {defaultValue=1, group=Stars, parameterName=Stars_Saturation, sliderMax=10, sortPriority=20}
  MaterialExpressionVectorParameter_2 = VectorParameter() {
   group=Stars
   parameterName=Stars_Coordinates
   sortPriority=0
   channelNames {a=Rotation Speed, b=Frequency, g="Axis Precession (Degrees)", r="Axis Tilt (Degrees)"}
   defaultValue=<rgba 23.4 0 0.09 10>
  }
  n2 = Multiply(
   A=@MaterialExpressionVectorParameter_2.Frequency
   B=TransformPosition(None=WorldPosition() {worldPositionShaderOffset=WPT_CameraRelative}.XYZ) {transformSourceType=TRANSFORMPOSSOURCE_TranslatedWorld}
  )
  n3 = Constant3Vector() {constant=<rgba 0 0 1 0>}
  n4 = Custom!(angle=@MaterialExpressionVectorParameter_2."Axis Precession (Degrees)") {
   code="float theta = radians(angle);\r\nfloat2x2 rot2d = {  cos(theta), sin(theta),\r\n                   -sin(theta), cos(theta) };\r\nreturn float3(mul(rot2d, float2(1.,0.)), 0.);"
   description=MakeDirection
   showCode=true
   inputs [{inputName=angle, input {expression=&self:MaterialExpressionVectorParameter_2, mask=1, maskG=1, outputIndex=2}}]
  }
  n5 = Constant3Vector()
  n6 = Add(
   A=@n3
   B=RotateAboutAxis(
    NormalizedRotationAxis=@n4
    RotationAngle=Divide(A=@MaterialExpressionVectorParameter_2."Axis Tilt (Degrees)") {constB=-360, desc=Degrees to 0-1}
    PivotPoint=@n5
    Position=@n3
   )
  )
  n7 = Multiply(A=Multiply(A=@MaterialExpressionVectorParameter_2.Rotation Speed) {constB=0.0001}, B=Time() {bIgnorePause=true, period=1})
  n8 = Add(A=@n2, B=RotateAboutAxis(NormalizedRotationAxis=Normalize(VectorInput=@n6), RotationAngle=Frac(None=@n7), PivotPoint=@n5, Position=@n2))
  Composite!(UVs=@n8, ColorStr=@n1) {subgraphName=Make Stars}
  n9 = Multiply(A=@n8) {constB=0.75}
  Composite(UVs=@n9, ColorStr=@n1) {subgraphName=Make Stars}
  n10 = Multiply(A=@n8) {constB=0.45}
  Composite(UVs=@n10, ColorStr=@n1) {subgraphName=Make Stars}
  n11 = Constant()
  n12 = ScalarParameter() {defaultValue=3.5401, parameterName=Param_3}
  n13 = ScalarParameter() {defaultValue=30.3601, parameterName=Param_4}
  n14 = MaterialFunctionCall(
   Plane Normal Axis=Constant3Vector() {constant=<rgba 0 0 1 0>}
   Plane X Axis=Constant3Vector() {constant=<rgba 1 0 0 0>}
   Plane Center=MaterialFunctionCall(X=@n11, Y=@n11, Z=MaterialFunctionCall!(Km=@n12) {materialFunction=/Engine/Functions/Engine_MaterialFunctions01/Units/Km-to-Cm}.Cm) {materialFunction=&MakeFloat3}.Result
   UV Scale=MaterialFunctionCall(Km=@n13) {materialFunction=&Km-to-Cm}.Cm
  ) {materialFunction=&VirtualPlaneCoordinates}
  n15 = ComponentMask(None=@n14.World Position) {g=true, r=true}
  Divide(
   A=@n15
   B=MaterialFunctionCall(Km=ScalarParameter() {defaultValue=25.9182, parameterName=Param_5}) {materialFunction=&Km-to-Cm}.Cm
  )
  n16 = ComponentMask(None=Normalize(VectorInput=@n14.World Position)) {g=true, r=true}
  n17 = ScalarParameter() {defaultValue=0.2, parameterName=Clouds_HorizonCurvature, sliderMax=1}
  n18 = Divide(A=DotProduct(A=@n16, B=@n16), B=Add(A=Multiply(A=@n17) {constB=0.01}))
  Divide(A=@n15, B=SquareRoot(None=OneMinus(None=@n18)) {desc=newZ})
  n19 = OneMinus(None=NamedRerouteUsage())
  MaterialExpressionReroute_3 = Reroute(None=@n8)
  n20 = Frac(None=@MaterialExpressionReroute_3.Coordinates)
  n21 = Length(None=Multiply(A=Subtract(A=@n20) {constB=0.5}) {constB=1.5}) {is=class MaterialExpressionLength}
  n22 = VectorNoise(World Position=@MaterialExpressionReroute_3.Coordinates) {quality=0}
  n23 = Multiply(A=Subtract(A=@n22) {constB=0.5}) {constB=1.2}
  n24 = Saturate(None=Subtract(B=Length(None=@n23)) {constA=0.333})
  n25 = Saturate(None=OneMinus(None=Divide(A=@n21, B=@n24)))
  n26 = MaterialFunctionCall(f3=@n22) {materialFunction=&AddComponents}
  n27 = BlackBody!(Temp=Floor(None=Multiply(A=@n26.f3) {constB=4500}))
  n28 = MaterialFunctionCall(Float3=@n27) {materialFunction=&BreakOutFloat3Components}
  n29 = Divide(A=@n27, B=Max(A=Max(A=@n28.R, B=@n28.R), B=@n28.B))
  n30 = Constant() {r=-1}
  MaterialExpressionReroute_9 = Reroute(None=@n1)
  n31 = Multiply(
   A=Multiply(A=@n25, B=@n25)
   B=Desaturation(None=LinearInterpolate(A=@n29, B=@n27) {constAlpha=0.04}, Fraction=Multiply(A=@MaterialExpressionReroute_9.ColorStr, B=@n30))
  )
  MaterialExpressionReroute_4 = Reroute(None=@n31)
  MaterialExpressionReroute_5 = Reroute(None=@n9)
  n32 = Frac(None=@MaterialExpressionReroute_5.Coordinates)
  n33 = Length(None=Multiply(A=Subtract(A=@n32) {constB=0.5}) {constB=1.5})
  n34 = VectorNoise(World Position=@MaterialExpressionReroute_5.Coordinates) {quality=0}
  n35 = Multiply(A=Subtract(A=@n34) {constB=0.5}) {constB=1.2}
  n36 = Saturate(None=Subtract(B=Length(None=@n35)) {constA=0.333})
  n37 = Saturate(None=OneMinus(None=Divide(A=@n33, B=@n36)))
  n38 = MaterialFunctionCall(f3=@n34) {materialFunction=&AddComponents}
  n39 = BlackBody(Temp=Floor(None=Multiply(A=@n38.f3) {constB=4500}))
  n40 = MaterialFunctionCall(Float3=@n39) {materialFunction=&BreakOutFloat3Components}
  n41 = Divide(A=@n39, B=Max(A=Max(A=@n40.R, B=@n40.R), B=@n40.B))
  n42 = Constant() {r=-1}
  MaterialExpressionReroute_1 = Reroute(None=@n1)
  n43 = Multiply(
   A=Multiply(A=@n37, B=@n37)
   B=Desaturation(None=LinearInterpolate(A=@n41, B=@n39) {constAlpha=0.04}, Fraction=Multiply(A=@MaterialExpressionReroute_1.ColorStr, B=@n42))
  )
  MaterialExpressionReroute_6 = Reroute(None=@n43)
  MaterialExpressionReroute_7 = Reroute(None=@n10)
  n44 = Frac(None=@MaterialExpressionReroute_7.Coordinates)
  n45 = Length(None=Multiply(A=Subtract(A=@n44) {constB=0.5}) {constB=1.5})
  n46 = VectorNoise(World Position=@MaterialExpressionReroute_7.Coordinates) {quality=0}
  n47 = Multiply(A=Subtract(A=@n46) {constB=0.5}) {constB=1.2}
  n48 = Saturate(None=Subtract(B=Length(None=@n47)) {constA=0.333})
  n49 = Saturate(None=OneMinus(None=Divide(A=@n45, B=@n48)))
  n50 = MaterialFunctionCall(f3=@n46) {materialFunction=&AddComponents}
  n51 = BlackBody(Temp=Floor(None=Multiply(A=@n50.f3) {constB=4500}))
  n52 = MaterialFunctionCall(Float3=@n51) {materialFunction=&BreakOutFloat3Components}
  n53 = Divide(A=@n51, B=Max(A=Max(A=@n52.R, B=@n52.R), B=@n52.B))
  n54 = Constant() {r=-1}
  MaterialExpressionReroute_0 = Reroute(None=@n1)
  n55 = Multiply(
   A=Multiply(A=@n49, B=@n49)
   B=Desaturation(None=LinearInterpolate(A=@n53, B=@n51) {constAlpha=0.04}, Fraction=Multiply(A=@MaterialExpressionReroute_0.ColorStr, B=@n54))
  )
  MaterialExpressionReroute_15 = Reroute(None=@n55)
  n56 = Multiply(
   A=ScalarParameter() {defaultValue=1, group=Stars, parameterName=Stars_Intensity, sliderMax=100, sortPriority=10}
   B=Add(A=Add(A=@MaterialExpressionReroute_4.Stars, B=@MaterialExpressionReroute_6.Stars), B=@MaterialExpressionReroute_15.Stars)
  )
  n57 = SkyAtmosphereViewLuminance!(WorldDirection)
  n58 = SkyAtmosphereLightDirection()
  n59 = DotProduct(A=CameraVectorWS(), B=Normalize(VectorInput=Multiply(A=@n58) {constB=-1}))
  n60 = Constant() {r=1}
  n61 = Multiply(A=ScalarParameter() {group=Sun, parameterName=Sun_Size, sliderMax=1, sortPriority=0}) {constB=0}
  n62 = ScalarParameter() {defaultValue=0.667, group=Sun, parameterName=Sun_DiskHardness, sliderMax=1, sortPriority=20}
  n63 = ScalarParameter() {defaultValue=10, group=Sun, parameterName=Sun_Intensity, sliderMax=100, sortPriority=10}
  n64 = SkyAtmosphereLightIlluminance!(World Position=CameraPositionWS())
  n65 = SkyAtmosphereLightIlluminanceOnGround()
  n66 = ShadingPathSwitch(Default=@n64, Deferred=@n64, Forward=@n65, Mobile=@n65)
  n67 = ScalarParameter() {defaultValue=1, group=Moon, parameterName=Moon_Intensity, sliderMax=100, sortPriority=50}
  n68 = VectorParameter() {
   group=Moon
   parameterName=Moon_BounceLight
   sortPriority=40
   channelNames {a=BounceIntensity, b="", g="", r=""}
   defaultValue=<rgba 0.5 0.5 0.5 0>
  }
  n69 = Multiply(A=Multiply(A=@n68, B=@n68), B=Saturate(None=NamedRerouteUsage()))
  n70 = SkyAtmosphereLightDirection() {lightIndex=1}
  n71 = SkyAtmosphereLightDirection()
  n72 = Subtract(A=@n70, B=@n71)
  n73 = Multiply(A=@n70, B=Constant3Vector() {constant=<rgba -1 -1 1 0>}) {constB=-1}
  n74 = VectorParameter() {
   group=Moon
   parameterName=Moon_PhaseCoords
   sortPriority=25
   channelNames {a=UseCelestialDirection, b="", g="", r=""}
   defaultValue=<rgba 0 0.92 0.08 0>
  }
  n75 = Saturate(None=@n74.UseCelestialDirection)
  n76 = MaterialFunctionCall!(
   Vector1=Normalize(VectorInput=LinearInterpolate(A=@n72, B=@n73, Alpha=@n75))
   Vector2=Normalize(VectorInput=ComponentMask(None=@n74) {b=true, g=true, r=true})
  ) {materialFunction=/Engine/Functions/Engine_MaterialFunctions02/Math/CreateThirdOrthogonalVector}
  n77 = ConstantBiasScale(None=ScalarParameter() {group=Moon, parameterName=Moon_PhaseOffset, sliderMax=1, sliderMin=-1, sortPriority=15})
  n78 = DotProduct(A=Multiply(A=@n71) {constB=-1}, B=@n73)
  n79 = Add(A=@n77, B=Add(A=Multiply(A=@n78) {constB=0.125}) {constB=0.375})
  n80 = Add(
   A=@n76.Vector2
   B=RotateAboutAxis(
    NormalizedRotationAxis=@n76.Vector3
    RotationAngle=LinearInterpolate(A=@n79, B=@n77, Alpha=@n75)
    PivotPoint=Constant3Vector()
    Position=@n76.Vector3
   )
  )
  n81 = WorldPosition() {worldPositionShaderOffset=WPT_CameraRelative}
  n82 = Length(None=@n81.XYZ)
  n83 = ScalarParameter() {defaultValue=5, group=Moon, parameterName=Moon_Size, sliderMax=10, sliderMin=1, sortPriority=10}
  n84 = Divide(A=Constant() {r=38440000}, B=@n83)
  n85 = MaterialFunctionCall!(
   Sphere Position=Divide(A=@n81.XYZ, B=Divide(A=@n82, B=@n84))
   Sphere Radius=Constant() {r=173700.4062}
   RayDirection=SkyAtmosphereLightDirection() {lightIndex=1}
   World position=CameraPositionWS()
  ) {materialFunction=/Engine/Functions/Engine_MaterialFunctions02/RayTracedSphere}
  n86 = MaterialFunctionCall!(BaseNormal=NamedRerouteUsage(), AdditionalNormal=@n85.Intersection Surface Normal) {materialFunction=/Engine/Functions/Engine_MaterialFunctions02/Utility/BlendAngleCorrectedNormals_WS}
  n87 = DotProduct(
   A=Multiply(
    A=Normalize(VectorInput=@n80)
    B=ScalarParameter() {defaultValue=2, group=Moon, parameterName=Moon_TerminatorHardness, sliderMax=10, sortPriority=30}
   )
   B=ShadingPathSwitch(Default=@n86.Result, Deferred=@n86.Result, Forward=@n85.Intersection Surface Normal, Mobile=@n85.Intersection Surface Normal)
  )
  n88 = Multiply(A=@n87, B=NamedRerouteUsage())
  n89 = Saturate(None=Multiply(A=@n87, B=Power(Base=@n88) {constExponent=0.125}))
  n90 = SkyAtmosphereLightDirection() {lightIndex=1}
  n91 = MaterialFunctionCall() {materialFunction=&CameraDirectionVector}
  n92 = Multiply(A=@n85.Ray Hit Bool, B=Step(X=DotProduct(A=@n90, B=@n91.Result)))
  n93 = SkyAtmosphereLightDirection() {lightIndex=1}
  n94 = CrossProduct(A=@n93, B=Constant3Vector() {constant=<rgba 0 0 -1 0>})
  n95 = Normalize(VectorInput=@n94)
  n96 = Divide(A=Normalize(VectorInput=WorldPosition() {worldPositionShaderOffset=WPT_CameraRelative}.XYZ), B=@n83) {constB=10}
  n97 = Constant() {r=220}
  n98 = Normalize(VectorInput=CrossProduct(A=@n94, B=@n93))
  n99 = DotProduct(
   A=@n85.Intersection Surface Normal
   B=MaterialFunctionCall() {materialFunction=&CameraDirectionVector}.Result
  )
  MaterialExpressionMaterialFunctionCall_31 = MaterialFunctionCall(X=Multiply(A=DotProduct(A=@n95, B=@n96), B=@n97), Y=Multiply(A=DotProduct(A=@n98, B=@n96), B=@n97), Z=@n99) {materialFunction=&MakeFloat3}
  MaterialExpressionVectorParameter_1 = VectorParameter() {
   group=Moon
   parameterName=Moon_Coordinates
   sortPriority=20
   channelNames {a=Offset Texture, b=Animate Rotation, g="Roll (degrees)", r="Pitch (degrees)"}
  }
  MaterialExpressionCustom_10 = Custom!(
   pos=@MaterialExpressionMaterialFunctionCall_31.Result
   pitch=@MaterialExpressionVectorParameter_1."Pitch (degrees)"
   roll=@MaterialExpressionVectorParameter_1."Pitch (degrees)"
  ) {
   code="float pr = radians(pitch);\r\nfloat rr = radians(roll);\r\nfloat cosPitch = cos(pr);\r\nfloat sinPitch = sin(pr);\r\nfloat cosRoll = cos(rr);\r\nfloat sinRoll = sin(rr);\r\n\r\nfloat3x3 rotMat = { cosRoll, -sinRoll * cosPitch,  sinRoll * sinPitch,\r\n                    sinRoll,  cosRoll * cosPitch, -cosRoll * sinPitch,\r\n                        0.0,            sinPitch,            cosPitch   };\r\n\r\nreturn mul(pos, rotMat);"
   description=Rot3D
   inputs [
    {inputName=pos, input {expression=&self:MaterialExpressionMaterialFunctionCall_31}}
    {inputName=pitch, input {expression=&self:MaterialExpressionVectorParameter_1, mask=1, maskR=1, outputIndex=1}}
    {inputName=roll, input {expression=&self:MaterialExpressionVectorParameter_1, mask=1, maskG=1, outputIndex=2}}
   ]
  }
  n100 = Custom!(pos=@MaterialExpressionCustom_10) {
   code="//ref\r\n//https://emildziewanowski.com/flat-planets/\r\nfloat width = sqrt(1. - pos.y * pos.y);\r\nfloat genX = pos.x/width*sign(pos.z);\r\nfloat2 gen = float2(genX, pos.y);\r\nfloat2 uv = asin(gen)/PI + float2(0.5,0.5);\r\nreturn uv;"
   description=GenerateSphericalUV
   outputType=CMOT_Float2
   inputs [{inputName=pos, input {expression=&self:MaterialExpressionCustom_10}}]
  }
  n101 = Constant() {r=1.7778}
  n102 = Frac(None=Multiply(A=Time(), B=@MaterialExpressionVectorParameter_1.Animate Rotation))
  n103 = AppendVector(A=Add(A=Multiply(A=@n101, B=@n102), B=@MaterialExpressionVectorParameter_1.Offset Texture), B=Constant())
  n104 = Multiply(
   A=Multiply(A=@n92, B=Add(A=@n100, B=@n103))
   B=Multiply(
    A=ScalarParameter() {defaultValue=1, group=Moon, parameterName=Moon_TextureTile, sliderMax=10, sortPriority=2}
    B=Constant2Vector() {g=1, r=0.5625}
   )
  ) {constB=5}
  n105 = TextureSampleParameter2D(UVs=@n104) {group=Moon, parameterName=Moon_Texture, sortPriority=0, texture=&/T_Moon_D}
  n106 = Multiply(A=Multiply(A=Add(A=@n69, B=@n89), B=@n105.RGB), B=NamedRerouteUsage())
  n107 = SkyAtmosphereLightIlluminance(World Position=CameraPositionWS()) {lightIndex=1}
  n108 = SkyAtmosphereLightIlluminanceOnGround() {lightIndex=1}
  n109 = Add(
   A=Multiply(A=SphereMask(A=@n59, B=@n60, Radius=@n61, Hardness=@n62) {attenuationRadius=0.001, hardnessPercent=0}, B=Multiply(A=@n63, B=@n66))
   B=Multiply(A=Multiply(A=@n67, B=@n106), B=ShadingPathSwitch(Default=@n107, Deferred=@n107, Forward=@n108, Mobile=@n108))
  )
  n110 = SkyAtmosphereDistantLightScatteredLuminance()
  n111 = ScalarParameter() {defaultValue=1, parameterName=C1Skylight}
  n112 = Constant()
  n113 = ScalarParameter() {defaultValue=3.5401, parameterName=Param_3}
  n114 = Add(
   A=ObjectPositionWS()
   B=MaterialFunctionCall(X=@n112, Y=@n112, Z=MaterialFunctionCall(Km=@n113) {materialFunction=&Km-to-Cm}.Cm) {materialFunction=&MakeFloat3}.Result
  )
  n115 = Multiply(
   A=ComponentMask(None=SkyAtmosphereAerialPerspective(World Position=@n114)) {b=true, g=true, r=true}
   B=VectorParameter() {defaultValue=<rgba 0.98 0.98 0.98 0>}.RGB
  )
  n116 = SkyAtmosphereLightIlluminance(World Position=@n114)
  n117 = Multiply(A=Multiply(A=ComponentMask(None=@n116) {b=true, g=true, r=true}, B=ScalarParameter() {defaultValue=0.2, parameterName=Param_2})) {constB=0, desc="mult by lighting (needs normal)"}
  n118 = CameraVectorWS()
  n119 = Subtract(A=WorldPosition() {worldPositionShaderOffset=WPT_CameraRelative}.XYZ, B=ObjectPositionWS() {originType=CameraRelative})
  n120 = ComponentMask(None=Divide(A=@n119, B=Length(None=@n119))) {g=true, r=true}
  n121 = Multiply(A=ScalarParameter() {defaultValue=0.2, parameterName=Clouds_HorizonCurvature, sliderMax=1}) {constB=0.01}
  n122 = OneMinus(None=Divide(A=DotProduct(A=@n120, B=@n120), B=Add(A=@n121)))
  n123 = Divide(
   A=@n120
   B=Subtract(A=SquareRoot(None=@n122) {desc=newZ}, B=ScalarParameter() {parameterName=Clouds_HorizonOffset, sliderMax=1, sliderMin=-1})
  )
  n124 = VectorParameter() {
   desc="This vector controls the strength of the wind along each world axis (x, y, or z), the value for which may be positive or negative.  The alpha channel can be used to uniformly increase or decrease the strength of the wind."
   group=00 - Cloud Layout
   parameterName=Layout_WindControls
   sortPriority=80
   channelNames {a=VelocityScale, b=Velocity.Z, g=Velocity.Y, r=Velocity.X}
   defaultValue=<rgba 1 1 0.5 2>
  }
  n125 = Transform(None=Multiply(A=@n124, B=MaterialFunctionCall!(Cm=@n124.VelocityScale) {materialFunction=/Engine/Functions/Engine_MaterialFunctions01/Units/Cm-to-Km}.Km)) {transformSourceType=TRANSFORMSOURCE_Local}
  n126 = Frac(None=Multiply(A=Time(), B=ComponentMask(None=@n125) {g=true, r=true}) {constB=0.01})
  n127 = TextureSample(UVs=Add(A=@n123, B=@n126)) {texture=/Engine/MaterialTemplates/Textures/T_Noise01}
  n128 = Multiply(
   A=ComponentMask(None=Transform(None=@n118) {transformSourceType=TRANSFORMSOURCE_World, transformType=TRANSFORM_Tangent}) {g=true, r=true}
   B=Multiply(A=Subtract(A=@n127.G) {constB=0.5}, B=ScalarParameter() {defaultValue=10})
  )
  n129 = Add(
   A=Multiply(A=Add(A=@n128, B=@n123), B=ScalarParameter() {defaultValue=0.2, parameterName=Clouds_Height, sliderMax=2}) {constB=0.1}
   B=AppendVector(
    A=ScalarParameter() {parameterName=Clouds_Offset.x, sliderMax=10, sliderMin=-10}
    B=ScalarParameter() {parameterName=Clouds_Offset.y, sliderMax=10, sliderMin=-10}
   )
  )
  LinearInterpolate(
   A=Add(A=Multiply(A=@n19, B=@n56) {desc=Stop stars from drawing on moon}, B=Add(A=@n57, B=@n109))
   B=Add(A=Multiply(A=@n110, B=@n111), B=Add(A=@n115, B=@n117))
   Alpha=TextureSample(UVs=Add(A=@n129, B=@n126)) {texture=&T_Noise01}.G
  )
  MaterialFunctionCall(Cm=@n13) {materialFunction=&Cm-to-Km}
  MaterialFunctionCall() {materialFunction=&Cm-to-Km}
  MaterialFunctionCall!(PixelNormal, In Depth, CUV In) {materialFunction=/Landmass/Landscape/OldPrototype_BP/Mats/MF/Inscattering_LUT}
  Multiply()
  NamedRerouteDeclaration(None=Transform(None=TextureSampleParameter2D(UVs=@n104) {group=Moon, parameterName=Moon_Normal, samplerType=SAMPLERTYPE_Normal, sortPriority=1, texture=&/T_Moon_N}.RGB)) {name=MoonNormWS, nodeColor=<rgba 0.9688 1 0 1>}
  NamedRerouteDeclaration(None=Multiply(A=@n92, B=@n105.A)) {name=MoonMask, nodeColor=<rgba 1 0 0.5156 1>}
  NamedRerouteDeclaration(None=@n99) {name=Dif, nodeColor=<rgba 1 0 0.5391 1>}
  Normalize(VectorInput=WorldPosition() {worldPositionShaderOffset=WPT_CameraRelative}.XYZ)
  PinBase!(Stars=@n31) {reroutePins [Stars {expression=&self:MaterialExpressionReroute_4}]}
  PinBase!() {reroutePins [
    Coordinates {expression=&self:MaterialExpressionReroute_3}
    ColorStr {expression=&self:MaterialExpressionReroute_9}
   ]}
  PinBase() {reroutePins [
    Coordinates {expression=&self:MaterialExpressionReroute_5}
    ColorStr {expression=&self:MaterialExpressionReroute_1}
   ]}
  PinBase!(Stars=@n43) {reroutePins [Stars {expression=&self:MaterialExpressionReroute_6}]}
  PinBase!() {reroutePins [
    Coordinates {expression=&self:MaterialExpressionReroute_7}
    ColorStr {expression=&self:MaterialExpressionReroute_0}
   ]}
  PinBase!(Stars=@n55) {reroutePins [Stars {expression=&self:MaterialExpressionReroute_15}]}
  n130 = Subtract(A=WorldPosition().Z, B=ComponentMask(None=ObjectPositionWS() {originType=CameraRelative}) {b=true})
  n131 = MaterialFunctionCall(Km=Multiply(A=ScalarParameter() {defaultValue=3, parameterName=Param_1}) {constB=-1}) {materialFunction=&Km-to-Cm}
  Saturate(None=Divide(A=Add(A=@n130, B=@n131.Cm) {constB=-4500}, B=@n131.Cm)) {desc=horizon gradient}
  ScalarParameter() {defaultValue=0.2, parameterName=Clouds_Height, sliderMax=10, sliderMin=0.1}
  SkyAtmosphereViewLuminance()
 ]
}
